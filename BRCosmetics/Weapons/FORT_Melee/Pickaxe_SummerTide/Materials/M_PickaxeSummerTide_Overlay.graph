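# Translucent unlit overlay material for the SummerTide pickaxe, faking glass with bubbles inside.
# Emissive sums skylight reflection, depth color and likely a sun highlight (Blinn/Phong, with collection-parameter fallback for light direction), then gets time-of-day tints.
# Opacity combines terms such as scene-depth thickness, which varies with distance, and Fresnel edges, limited by a tech-art mask.
# MF_Bubbles_OUT adds bubbles to emissive before the accessory tech-art overrides.
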Material /BRCosmetics/Weapons/FORT_Melee/Pickaxe_SummerTide/Materials/M_PickaxeSummerTide_Overlay
graph {
 properties {
  bAutomaticallySetUsageInEditor=false
  bUseMaterialAttributes=true
  bUsedWithMeshDeformer=true
  bUsedWithSkeletalMesh=true
  blendMode=BLEND_Translucent
  pixelDepthOffsetMode=PDOM_Legacy
  shadingModel=MSM_Unlit
  usedShadingModels=Unlit
 }
 nodes [
  Constant()
  n1 = NamedRerouteUsage()
  n2 = VectorParameter() {group=FakeGlass, parameterName=EveningTint, defaultValue=<rgba 1 0.6167 0.9182 1>}
  n3 = VectorParameter() {group=FakeGlass, parameterName=NightTint, defaultValue=<rgba 0.5417 0.5626 1 1>}
  n4 = SetMaterialAttributes!(
   MaterialAttributes
   Emissive Color=MaterialFunctionCall(Color_Morning=@n1, Color_Day=@n1, Color_Evening=Multiply(A=@n1, B=@n2.RGB), Color_Night=Multiply(A=@n1, B=@n3.RGB)) {materialFunction=&Color_Time_of_Day}.Result
   Opacity=NamedRerouteUsage()
  ) {attributeSetTypes [#1, #2]}
  n5 = GetMaterialAttributes(None=@n4) {attributeGetTypes [#1]}
  n6 = MaterialFunctionCall!(Material=@n4, BubblesMask, SlurpNormalDistortionMask, Pre Skinned Position, TechArtMask, UseBubble, UseTriPlanarBubbles) {materialFunction=/Game/Global/Materials/MaterialGym/Materials/Bubbles/MF_Bubbles_OUT}
  n7 = Multiply(
   A=Multiply(A=Multiply(A=@n6.Bubble_OUT, B=@n6.Bubble_OUT), B=NamedRerouteUsage())
   B=ScalarParameter() {defaultValue=0.7, group=FakeGlass_Water, parameterName=BubbleIntensity}
  )
  MaterialFunctionCall(In=SetMaterialAttributes!(MaterialAttributes=@n5.MaterialAttributes, Emissive Color=Add(A=@n5.Emissive Color, B=@n7)) {attributeSetTypes [#1]}) {materialFunction=&MF_TechArtOverrides_accessories}
  n8 = ComponentMask(None=Custom() {code="return View.DirectionalLightColor;", description=DirectionalLightColor, outputType=CMOT_Float4, inputs []}) {b=true, g=true, r=true}
  n9 = Custom() {code="return View.DirectionalLightDirection;", description=DirectionalLightDirection, inputs []}
  n10 = CollectionParameter() {
   collection=&FortniteMaterialParameters
   parameterName=SunlightDirectionalVector
  }
  n11 = ComponentMask(None=@n10) {b=true, g=true, r=true}
  n12 = If(A=DotProduct(A=@n9, B=@n9), "A > B"=@n9, "A == B"=@n11, "A < B"=@n11) {desc=check if DirectionalLight exists}
  n13 = ComponentMask(None=NamedRerouteDeclaration(None=If(A=@n8, "A > B"=@n12, "A == B"=@n11, "A < B"=@n11) {desc=check if DirectionalLight color is Black}) {name=LightVec, nodeColor=<rgba 0 1 0.9297 1>}) {b=true, g=true, r=true}
  n14 = MaterialFunctionCall(Float3=@n13) {materialFunction=&BreakOutFloat3Components}
  n15 = MaterialFunctionCall(X=@n14.R, Y=@n14.R, Z=Max(A=@n14.B) {constB=0.2}) {materialFunction=&MakeFloat3}
  n16 = MaterialFunctionCall(Float3=@n13) {materialFunction=&BreakOutFloat3Components}
  n17 = MaterialFunctionCall(X=@n16.R, Y=@n16.R, Z=Max(A=@n16.B) {constB=0.9}) {materialFunction=&MakeFloat3}
  n18 = ScalarParameter() {group=FakeGlass, parameterName=IsFrontEnd}
  n19 = ComponentMask(None=LinearInterpolate(
    A=Normalize(VectorInput=@n15.Result)
    B=Normalize(VectorInput=@n17.Result)
    Alpha=NamedRerouteDeclaration(None=@n18) {name=FrontEnd, nodeColor=<rgba 0.8047 1 0 1>}
   )) {b=true, g=true, r=true}
  n20 = Normalize(VectorInput=Add(A=CameraVectorWS(), B=@n19))
  n21 = Normalize(VectorInput=PixelNormalWS())
  n22 = ReflectionVectorWS()
  n23 = Normalize(VectorInput=@n19)
  n24 = Max(A=LinearInterpolate(
    A=DotProduct(A=@n20, B=@n21) {desc=Blinn}
    B=DotProduct(A=@n22, B=@n23) {desc=Phong}
    Alpha=ScalarParameter() {defaultValue=1, group=FakeGlass, parameterName=BlinnOrPhong}
   )) {constB=0}
  n25 = ScalarParameter() {defaultValue=20, group=FakeGlass, parameterName=SpecPow}
  NamedRerouteDeclaration(None=SmoothStep(Max=ScalarParameter() {defaultValue=0.7, group=FakeGlass, parameterName=SpecMax}, Value=Power(Base=@n24, Exp=@n25)) {constMax=0.5}) {name=HighLight, nodeColor=<rgba 0.3359 1 0 1>}
  n26 = LinearInterpolate(B=ScalarParameter() {defaultValue=0.3, group=FakeGlass_Water, parameterName=MinAbsorptionValue_Depth}, Alpha=NamedRerouteUsage()) {constB=0.3}
  n27 = Subtract(
   A=SceneDepth().R
   B=MaterialFunctionCall!(World Position) {materialFunction=/Engine/Functions/Engine_MaterialFunctions02/Utility/DepthFromWorldPosition}.Result
  )
  n28 = MaterialFunctionCall!(
   Blend Range=ScalarParameter() {defaultValue=300, group=FakeGlass, parameterName=DepthThickness_BlendRange}
   Start Offset=ScalarParameter() {defaultValue=-20, group=FakeGlass, parameterName=DepthThickness_StartOffset}
  ) {materialFunction=/Engine/Functions/Engine_MaterialFunctions03/Particles/Distance_Blend}
  n29 = LinearInterpolate(
   A=ScalarParameter() {defaultValue=100, group=FakeGlass, parameterName=DepthThickness}
   B=ScalarParameter() {defaultValue=5, group=FakeGlass, parameterName=DepthThickness_Far}
   Alpha=@n28.Result
  )
  n30 = Power(Base=Saturate(None=Divide(A=@n27, B=@n29)), Exp=ScalarParameter() {defaultValue=1, group=FakeGlass, parameterName=DepthPower})
  n31 = ScalarParameter() {defaultValue=1, group=FakeGlass, parameterName=DepthAmount}
  n32 = ScalarParameter() {defaultValue=1, group=FakeGlass, parameterName=DepthAmount_Far}
  n33 = NamedRerouteUsage()
  NamedRerouteDeclaration(None=Multiply(A=LinearInterpolate(A=@n26, Alpha=@n30), B=LinearInterpolate(A=@n31, B=@n32, Alpha=@n33))) {name=DepthOpacity, nodeColor=<rgba 1 0.2578 0 1>}
  n34 = SkyLightEnvMapSample(
   Direction=ReflectionVectorWS(CustomWorldNormal=PixelNormalWS())
   Roughness=ScalarParameter() {defaultValue=0.22, group=FakeGlass, parameterName=Roughness}
  )
  n35 = ScalarParameter() {defaultValue=1, group=FakeGlass, parameterName=ReflectionStrength}
  n36 = Multiply(
   A=Multiply(A=VectorParameter() {group=FakeGlass, parameterName=ReflectColor, defaultValue=<rgba 1 1 1 1>}.RGB, B=Multiply(A=@n34, B=@n35))
   B=NamedRerouteUsage()
  )
  n37 = NamedRerouteUsage()
  n38 = VectorParameter() {group=FakeGlass, parameterName=DepthColor, defaultValue=<rgba 1 0 0.0413 1>}
  n39 = NamedRerouteUsage()
  NamedRerouteDeclaration(None=Add(A=Add(A=@n36, B=@n37), B=Multiply(A=@n38.RGB, B=@n39))) {name=EmissiveOUT, nodeColor=<rgba 1 0 0.5625 1>}
  n40 = ScalarParameter() {defaultValue=0.7, parameterName=OuterEdgeThickness}
  n41 = Subtract(A=@n40, B=ScalarParameter() {defaultValue=0.05, parameterName=OuterEdgeFade})
  n42 = Fresnel() {baseReflectFraction=0, exponent=1}
  n43 = Subtract(A=@n41, B=ScalarParameter() {parameterName=InnerEdgeFadeBias})
  n44 = Multiply(A=OneMinus(None=SmoothStep(Min=@n41, Max=@n40, Value=@n42)), B=SmoothStep(Min=Constant(), Max=Saturate(None=@n43), Value=@n42))
  n45 = ScalarParameter() {defaultValue=2, parameterName=Exp}
  NamedRerouteDeclaration(None=Saturate(None=Power(Base=@n44, Exp=@n45))) {name=FresnelOpacity, nodeColor=<rgba 0.8516 0 1 1>}
  NamedRerouteDeclaration(None=@n28.Result) {name=Distance_BlendAlpha, nodeColor=<rgba 0.0781 0 1 1>}
  n46 = TextureSampleParameter2D() {
   group=FakeGlass_Water
   parameterName=TechArtMask
   samplerType=SAMPLERTYPE_Masks
   texture=/BRCosmetics/Weapons/FORT_Melee/Pickaxe_SummerTide/Textures/T_SummerTideGlassMask
  }
  NamedRerouteDeclaration(None=ChannelMaskParameter(None=@n46.RGB) {group=FakeGlass, parameterName=UseFakeGlass}.R) {name=TechArtMask, nodeColor=<rgba 1 0.9375 0 1>}
  n47 = LinearInterpolate(
   A=ScalarParameter() {group=FakeGlass, parameterName=OpacityMin}
   B=ScalarParameter() {defaultValue=1, group=FakeGlass, parameterName=OpacityMax}
   Alpha=NamedRerouteUsage()
  )
  n48 = Add(A=NamedRerouteUsage(), B=Add(A=Saturate(None=@n47), B=NamedRerouteUsage()))
  NamedRerouteDeclaration(None=LinearInterpolate(B=Saturate(None=@n48), Alpha=NamedRerouteUsage()) {constB=0}) {name=Opacity_OUT, nodeColor=<rgba 0.6875 0 1 1>}
 ]
}
