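# Masked, two-sided material for the SummerTide pickaxe that fakes glass by rendering only back faces, using the shared 2023 character PreFX/PostFX functions.
# TwoSidedSign zeroes the opacity mask on front faces, while ShadowReplace leaves the mask intact for shadows.
# A tinted, smoothstepped term from a named reroute, apparently Fresnel, is added to emissive.
# The second emissive-color add and extra Fresnel shaping chains feed no output.
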Material /BRCosmetics/Weapons/FORT_Melee/Pickaxe_SummerTide/Materials/M_Pickaxe_SummerTide_GlassBack
graph {
 properties {
  bUseMaterialAttributes=true
  bUsedWithMeshDeformer=true
  bUsedWithSkeletalMesh=true
  blendMode=BLEND_Masked
  pixelDepthOffsetMode=PDOM_Legacy
  shadingModel=MSM_FromMaterialExpression
  twoSided=true
  usedShadingModels=Default Lit | Subsurface | Clear Coat | Subsurface Profile | Cloth
 }
 nodes [
  n1 = MaterialFunctionCall() {materialFunction=&MF_CharacterParent_2023_PreFX}
  n2 = GetMaterialAttributes(None=@n1.Result) {attributeGetTypes [#1, #2, #3]}
  n3 = VectorParameter() {group=FakeGlass, parameterName=GlassBackFaceFresenelColor, defaultValue=<rgba 1 1 1 1>}
  n4 = NamedRerouteUsage()
  n5 = Add(A=@n2.Emissive Color, B=SmoothStep(Value=Multiply(A=@n3.RGB, B=@n4)) {constMin=0.1})
  Add(
   A=@n5
   B=Multiply(A=VectorParameter() {group=FakeGlass, parameterName=GlassBackFaceFresenelColor_EM, defaultValue=<rgba 1 1 1 1>}.RGB, B=NamedRerouteUsage())
  )
  n6 = Sign(None=TwoSidedSign())
  n7 = ShadowReplace(Default=LinearInterpolate(Alpha=Saturate(None=@n6)) {constA=1, constB=0}, Shadow=Constant() {r=1})
  MaterialFunctionCall(
   Material=SetMaterialAttributes!(
    MaterialAttributes=@n2.MaterialAttributes
    Opacity Mask=Multiply(A=@n2.Opacity Mask, B=@n7)
    Base Color=@n2.MaterialAttributes
    Emissive Color=@n5
   ) {attributeSetTypes [#1, #2, #3]}
   M=@n1.M
   Pre Skinned Local Position=@n1.M
   Distance Blend=@n1.M
   Fresnel=@n1.M
  ) {materialFunction=&MF_CharacterParent_2023_PostFX}
  n8 = Saturate(None=MaterialFunctionCall(
    Normal Vector=PixelNormalWS()
    Power=ScalarParameter() {defaultValue=2, group=FakeGlass, parameterName=FresnelExponent_EM, sortPriority=10}
    Use Cheap contrast=StaticBool()
    Cheap contrast dark=ScalarParameter() {group=FakeGlass, parameterName=FresnelDarkLevel_EM, sortPriority=11}
    Cheap contrast bright=ScalarParameter() {defaultValue=1, group=FakeGlass, parameterName=FresnelBrightLevel_EM, sortPriority=12}
   ) {materialFunction=&Fresnel_Function}.Result)
  NamedRerouteDeclaration(None=@n8) {name=Fresnel_02, nodeColor=<rgba 0 1 0.4844 1>}
  n9 = PixelNormalWS()
  n10 = ScalarParameter() {defaultValue=5, group=FakeGlass, parameterName=FresnelExponent, sortPriority=10}
  n11 = StaticBool()
  n12 = ScalarParameter() {group=FakeGlass, parameterName=FresnelDarkLevel, sortPriority=11}
  n13 = ScalarParameter() {defaultValue=1, group=FakeGlass, parameterName=FresnelBrightLevel, sortPriority=12}
  NamedRerouteDeclaration(None=Saturate(None=MaterialFunctionCall(Normal Vector=@n9, Power=@n10, Use Cheap contrast=@n11, Cheap contrast dark=@n12, Cheap contrast bright=@n13) {materialFunction=&Fresnel_Function}.Result)) {name=Fresnel_01, nodeColor=<rgba 0 1 0.4844 1>}
  Power(Base=@n8) {constExponent=3}
  n14 = Add(A=Sine(None=@n8), B=ScalarParameter() {group=FakeGlass, parameterName=Fresnel_EM_Offset})
  n15 = Fresnel() {baseReflectFraction=0, exponent=3}
  SmoothStep(Value=Multiply(A=Saturate(None=@n14), B=Multiply(A=@n8, B=@n15))) {constMin=0.1}
 ]
}
