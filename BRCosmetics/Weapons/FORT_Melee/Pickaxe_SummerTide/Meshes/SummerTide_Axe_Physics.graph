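# Physics asset for the SummerTide pickaxe, a two-handed melee cosmetic with a duck-hammer head.
# Capsule bodies cover the handle, root, hammer base and a duck wobbler bone.
# Handle and hammer base are constrained to the root with moderate swing and twist limits.
# The wobbler hangs off the hammer base with tight swing limits and zero twist, restricting it to a slight wobble.
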PhysicsAsset /BRCosmetics/Weapons/FORT_Melee/Pickaxe_SummerTide/Meshes/SummerTide_Axe_Physics
bodies {
 Melee_TwoHanded_Handle [{shapeType=Capsule, length=56.4402, radius=21.97, center=<0.7723 -3.6657 20.6038>, rotation=<rot 0.8882 -90 0.0215>}]
 b_AR_Root [{shapeType=Capsule, length=0.505, radius=0.505, rotation=<rot 0 -90 0>}]
 base_duckhammer [{shapeType=Capsule, length=30.2115, radius=11.5609, center=<2.825 1.8153 0.003>, rotation=<rot 0 170.145 -89.9753>}]
 duck_wobbler [{
   shapeType=Capsule
   length=11.4082
   radius=7.7198
   shapeName=duck_wobbler_capsule
   center=<5.856 -0.026 4.3371>
   rotation=<rot 0.1458 -90 20.7136>
  }]
}
constraints [
 {bone1Name=Melee_TwoHanded_Handle, bone2Name=b_AR_Root, swing1LimitDegrees=45, swing2LimitDegrees=45, twistLimitDegrees=45}
 {bone1Name=base_duckhammer, bone2Name=b_AR_Root, swing1LimitDegrees=45, swing2LimitDegrees=45, twistLimitDegrees=45}
 {bone1Name=duck_wobbler, bone2Name=base_duckhammer, swing1LimitDegrees=3, swing2LimitDegrees=8, twistLimitDegrees=0}
]
